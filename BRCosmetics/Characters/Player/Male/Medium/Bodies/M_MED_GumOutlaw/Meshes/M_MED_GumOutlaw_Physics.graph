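# Physics asset for the GumOutlaw medium male player body cosmetic, set up for a dangling two-strand belt.
# Capsules on the pelvis, spine and thighs form a simplified body around the belt bones.
# Front and side belt chains hang from spine_04 and spine_02 through free-rotation joints, with narrow swing limits and little or no twist along each chain.
# A free-rotation constraint joins the two chain tips, closing the belt into a loop.
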PhysicsAsset /BRCosmetics/Characters/Player/Male/Medium/Bodies/M_MED_GumOutlaw/Meshes/M_MED_GumOutlaw_Physics
bodies {
 Base_belt_Side [{shapeType=Capsule, length=2, radius=2, shapeName=Base_belt_Side_capsule}]
 Base_belt_front [{shapeType=Capsule, length=2, radius=2, shapeName=Base_belt_front_capsule}]
 Base_belt_front_03 [{shapeType=Capsule, length=2, radius=2, shapeName=Base_belt_front_03_capsule}]
 Dyn_belt_Side_01 [{shapeType=Capsule, length=3.0049, radius=2.9304, shapeName=Dyn_belt_Side_01_capsule, center=<4.0656 0 1.9229>, rotation=<rot 90 0 0>}]
 Dyn_belt_Side_02 [{
   shapeType=Capsule
   length=3.5974
   radius=3.5931
   shapeName=Dyn_belt_Side_02_capsule
   center=<4.204 0.9267 1.8037>
   rotation=<rot 87.8056 -0.0006 -0.0005>
  }]
 Dyn_belt_Side_03 [{
   shapeType=Capsule
   length=13.1651
   radius=6.4843
   shapeName=Dyn_belt_Side_03_capsule
   center=<5.7301 0.4273 3.044>
   rotation=<rot 86.0645 -0.0017 -0.0016>
  }]
 Dyn_belt_front_01 [{shapeType=Capsule, length=5.6986, radius=3.2113, shapeName=Dyn_belt_front_01_capsule, center=<4.8708 0 2.2497>, rotation=<rot 90 0 0>}]
 Dyn_belt_front_02 [{shapeType=Capsule, length=5.176, radius=3.2068, shapeName=Dyn_belt_front_02_capsule, center=<4.9166 0 1.9474>, rotation=<rot 90 0 0>}]
 Dyn_belt_front_03 [{
   shapeType=Capsule
   length=13.1651
   radius=6.4843
   shapeName=Dyn_belt_Side_03_capsule
   center=<5.7301 0.4273 3.044>
   rotation=<rot 86.0645 -0.0017 -0.0016>
  }]
 pelvis [{shapeType=Capsule, length=9.7862, radius=12.0095, center=<-1.9373 0 0>}]
 spine_01 [{shapeType=Capsule, length=9.766, radius=11.6973, shapeName=spine_01_capsule, rotation=<rot 0 -90 0>}]
 spine_02 [{shapeType=Capsule, length=6.8298, radius=12.2921}]
 spine_03 [{shapeType=Capsule, length=6.8989, radius=13.2156}]
 spine_04 [{shapeType=Capsule, length=13.6566, radius=10.2329}]
 thigh_l [{shapeType=Capsule, length=21.1602, radius=8.6922, center=<-19.0171 0 0>, rotation=<rot 90 0 0>}]
 thigh_r [{shapeType=Capsule, length=20.6233, radius=9.2291, center=<19.0512 0.003 -0.3918>, rotation=<rot 85 -175.0002 0>}]
}
constraints [
 {bone1Name=spine_02, bone2Name=spine_01, swing1LimitDegrees=45, swing2LimitDegrees=45, twistLimitDegrees=45}
 {bone1Name=spine_03, bone2Name=spine_02, swing1LimitDegrees=45, swing2LimitDegrees=45, twistLimitDegrees=45}
 {bone1Name=spine_04, bone2Name=spine_03, swing1LimitDegrees=45, swing2LimitDegrees=45, twistLimitDegrees=45}
 {bone1Name=thigh_l, bone2Name=pelvis, swing1LimitDegrees=45, swing2LimitDegrees=45, twistLimitDegrees=45}
 {
  bone1Name=spine_01
  bone2Name=pelvis
  swing1LimitDegrees=45
  swing1Motion=Free
  swing2LimitDegrees=45
  swing2Motion=Free
  twistLimitDegrees=45
  twistMotion=Free
 }
 {bone1Name=thigh_r, bone2Name=pelvis, swing1LimitDegrees=45, swing2LimitDegrees=45, twistLimitDegrees=45}
 {
  bone1Name=Base_belt_front
  bone2Name=spine_04
  swing1LimitDegrees=45
  swing1Motion=Free
  swing2LimitDegrees=45
  swing2Motion=Free
  twistLimitDegrees=45
  twistMotion=Free
 }
 {bone1Name=Dyn_belt_front_01, bone2Name=Base_belt_front, swing1LimitDegrees=15, swing2LimitDegrees=15, twistLimitDegrees=5}
 {bone1Name=Dyn_belt_front_02, bone2Name=Dyn_belt_front_01, swing1LimitDegrees=10, swing2LimitDegrees=10, twistLimitDegrees=0}
 {bone1Name=Base_belt_front_03, bone2Name=Dyn_belt_front_02, swing1LimitDegrees=5, swing2LimitDegrees=5, twistLimitDegrees=0}
 {bone1Name=Dyn_belt_front_03, bone2Name=Base_belt_front_03, swing1LimitDegrees=8, swing2LimitDegrees=8, twistLimitDegrees=0}
 {
  bone1Name=Base_belt_Side
  bone2Name=spine_02
  swing1LimitDegrees=45
  swing1Motion=Free
  swing2LimitDegrees=45
  swing2Motion=Free
  twistLimitDegrees=45
  twistMotion=Free
 }
 {bone1Name=Dyn_belt_Side_01, bone2Name=Base_belt_Side, swing1LimitDegrees=12, swing2LimitDegrees=12, twistLimitDegrees=0}
 {bone1Name=Dyn_belt_Side_02, bone2Name=Dyn_belt_Side_01, swing1LimitDegrees=10, swing2LimitDegrees=10, twistLimitDegrees=0}
 {bone1Name=Dyn_belt_Side_03, bone2Name=Dyn_belt_Side_02, swing1LimitDegrees=8, swing2LimitDegrees=8, twistLimitDegrees=0}
 {
  bone1Name=Dyn_belt_Side_03
  bone2Name=Dyn_belt_front_03
  swing1LimitDegrees=45
  swing1Motion=Free
  swing2LimitDegrees=45
  swing2Motion=Free
  twistLimitDegrees=45
  twistMotion=Free
 }
]
