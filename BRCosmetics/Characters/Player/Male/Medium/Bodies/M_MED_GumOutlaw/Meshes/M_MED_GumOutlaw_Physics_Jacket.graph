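# Physics asset for the GumOutlaw medium male body's jacket, giving capsule bodies to the dynamic jacket bones and to the pelvis, lower spine and both thighs.
# Front and side flaps are two-link chains from small base bodies, while left and right back flaps attach directly to the pelvis.
# The front base is constrained to the left thigh and the side base to the pelvis.
# Flap joints allow narrow swing and zero twist.
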PhysicsAsset /BRCosmetics/Characters/Player/Male/Medium/Bodies/M_MED_GumOutlaw/Meshes/M_MED_GumOutlaw_Physics_Jacket
bodies {
 Base_jacket_Front [{shapeType=Capsule, length=2, radius=2, shapeName=Base_jacket_Front_capsule}]
 Base_jacket_side [{shapeType=Capsule, length=0.505, radius=0.505, shapeName=Base_jacket_side_capsule, rotation=<rot 0 -90 0>}]
 Dyn_jacket_Front_01 [{shapeType=Capsule, length=5.0487, radius=3.4374, shapeName=Dyn_jacket_Front_01_capsule, center=<3.5407 0 3.2078>, rotation=<rot 0 0 90>}]
 Dyn_jacket_Front_02 [{shapeType=Capsule, length=4.3202, radius=3.9174, shapeName=Dyn_jacket_Front_02_capsule, center=<3.5407 0 3.2078>, rotation=<rot 0 0 90>}]
 Dyn_jacket_back_L_01 [{
   shapeType=Capsule
   length=4.0245
   radius=3.5809
   shapeName=Dyn_jacket_back_L_01_capsule
   center=<4.4891 0.1894 2.6745>
   rotation=<rot 2.5755 -9.6658 104.7822>
  }]
 Dyn_jacket_back_R_01 [{
   shapeType=Capsule
   length=4.0245
   radius=3.5809
   shapeName=Dyn_jacket_back_L_01_capsule
   center=<3.0108 -0.5534 3.0438>
   rotation=<rot 3.0197 -4.6796 -99.9742>
  }]
 Dyn_jacket_side_01 [{
   shapeType=Capsule
   length=4.2069
   radius=3.3985
   shapeName=Dyn_jacket_side_01_capsule
   center=<3.9373 -0.0338 1.9785>
   rotation=<rot 0 0.0001 85>
  }]
 Dyn_jacket_side_02 [{
   shapeType=Capsule
   length=4.2069
   radius=3.3985
   shapeName=Dyn_jacket_side_02_capsule
   center=<3.9373 -0.0338 1.9785>
   rotation=<rot 0 0.0001 85>
  }]
 pelvis [{shapeType=Capsule, length=7.8031, radius=10.8849}]
 spine_01 [{shapeType=Capsule, length=9.7036, radius=10.2701, shapeName=spine_01_capsule}]
 thigh_l [{shapeType=Capsule, length=21.0447, radius=8.6359, center=<-20.2012 0.3607 0>, rotation=<rot 90 179.9803 179.9808>}]
 thigh_r [{shapeType=Capsule, length=21.4419, radius=8.2387, center=<20.5718 -0.3607 0>, rotation=<rot 90 89.6476 -90.3516>}]
}
constraints [
 {bone1Name=thigh_l, bone2Name=pelvis, swing1LimitDegrees=45, swing2LimitDegrees=45, twistLimitDegrees=45}
 {
  bone1Name=spine_01
  bone2Name=pelvis
  swing1LimitDegrees=45
  swing1Motion=Free
  swing2LimitDegrees=45
  swing2Motion=Free
  twistLimitDegrees=45
  twistMotion=Free
 }
 {bone1Name=thigh_r, bone2Name=pelvis, swing1LimitDegrees=45, swing2LimitDegrees=45, twistLimitDegrees=45}
 {bone1Name=Dyn_jacket_side_01, bone2Name=Base_jacket_side, swing1LimitDegrees=10, swing2LimitDegrees=10, twistLimitDegrees=0}
 {bone1Name=Dyn_jacket_side_02, bone2Name=Dyn_jacket_side_01, swing1LimitDegrees=8, swing2LimitDegrees=8, twistLimitDegrees=0}
 {bone1Name=Dyn_jacket_back_L_01, bone2Name=pelvis, swing1LimitDegrees=10, swing2LimitDegrees=15, twistLimitDegrees=0}
 {bone1Name=Dyn_jacket_back_R_01, bone2Name=pelvis, swing1LimitDegrees=10, swing2LimitDegrees=15, twistLimitDegrees=0}
 {
  bone1Name=Base_jacket_Front
  bone2Name=thigh_l
  swing1LimitDegrees=45
  swing1Motion=Free
  swing2LimitDegrees=45
  swing2Motion=Free
  twistLimitDegrees=45
  twistMotion=Free
 }
 {bone1Name=Dyn_jacket_Front_01, bone2Name=Base_jacket_Front, swing1LimitDegrees=7, swing2LimitDegrees=7, twistLimitDegrees=0}
 {bone1Name=Dyn_jacket_Front_02, bone2Name=Dyn_jacket_Front_01, swing1LimitDegrees=5, swing2LimitDegrees=5, twistLimitDegrees=0}
 {
  bone1Name=Base_jacket_side
  bone2Name=pelvis
  swing1LimitDegrees=45
  swing1Motion=Free
  swing2LimitDegrees=45
  swing2Motion=Free
  twistLimitDegrees=45
  twistMotion=Free
 }
]
