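# Character-part AnimBlueprint for the M_MED_GumOutlaw body cosmetic, adding procedural secondary motion to the pose copied from the attached parent mesh.
# The copied pose passes through a Control Rig fed state flags, then body and jacket RigidBody simulations.
# AnimDynamics springs on blade, zippers and lapel follow, then additive belt-bone offsets.
# The EventGraph updates crouch, DBNO, emote, LOD and velocity flags, a gravity override via a deprecated function, and interpolated front-end blend values.
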AnimBlueprint /BRCosmetics/Characters/Player/Male/Medium/Bodies/M_MED_GumOutlaw/Meshes/M_MED_GumOutlaw_AnimBP
target_skeleton=&/M_MED_GumOutlaw_Skeleton
graphs [
 AnimGraph {nodes [
   n1 = Control Rig!(
    Source=Copy Pose From Mesh() {bUseAttachedParent=true}
    Alpha
    DBNO:bool=Get bIsDBNO()
    Playing Emote:bool=Get Playing Emote() -> (Playing Emote:bool)
    RBAN off:bool=Get RBAN off() -> (RBAN off:bool)
    LOD2:bool=Get LOD2() -> (LOD2:bool)
    velociry forvard:bool=Get velociry forvard() -> (velociry forvard:bool)
    isInFrontEnd:bool=Get bIsInFrontEnd()
   ) -> (Pose) {
    alphaBoolBlend {blendOption=Linear}
    controlRigAssetReference {blueprintRigClass=&/M_MED_GumOutlaw_CtrlRig.M_MED_GumOutlaw_CtrlRig_C}
   }
   n2 = RigidBody!(ComponentPose=Local To Component(LocalPose=@n1), bAlphaBoolEnabled:bool, OverrideWorldGravity:Vector="0,0,-980", ExternalForce) -> (Pose) {
    alphaInputType=Bool
    bClampLinearTranslationLimitToRefPose=true
    bOverrideWorldGravity=true
    overridePhysicsAsset=&/M_MED_GumOutlaw_Physics
    simulationSpace=BaseBoneSpace
    alphaBoolBlend {blendOption=Linear}
    baseBoneRef {boneName=spine_02}
    overrideWorldGravity {z=-980}
    simSpaceSettings {
     maxAngularAcceleration=500
     maxAngularVelocity=3
     maxLinearAcceleration=400
     maxLinearVelocity=250
     velocityScaleZ=0
     worldAlpha=1
     externalLinearDragV=<0.1 0.1>
    }
   }
   n3 = RigidBody!(ComponentPose=@n2, Alpha:float, OverrideWorldGravity="0,0,-980", ExternalForce) -> (Pose) {
    bClampLinearTranslationLimitToRefPose=true
    bOverrideWorldGravity=true
    overridePhysicsAsset=&/M_MED_GumOutlaw_Physics_Jacket
    simulationSpace=BaseBoneSpace
    alphaBoolBlend {blendOption=Linear}
    baseBoneRef {boneName=pelvis}
    componentLinearAccScale {x=0.35, z=0.8}
    componentLinearVelScale {x=0.35, z=0.8}
    overrideWorldGravity {z=-980}
    simSpaceSettings {
     maxAngularAcceleration=500
     maxAngularVelocity=3
     maxLinearAcceleration=400
     maxLinearVelocity=250
     velocityScaleZ=0
     worldAlpha=1
     externalLinearDragV=<0.1 0.1>
    }
   }
   n4 = Anim Dynamics - Bone: Base_blade(ComponentPose=@n3, Alpha=1, ExternalForce="0,0,0") {
    angularDampingOverride=1
    angularSpringConstant=300
    bAngularSpring=true
    bOverrideAngularDamping=true
    bUseGravityOverride=true
    simulationSpace=BoneRelative
    alphaBoolBlend {blendOption=Linear}
    boundBone {boneName=Base_blade}
    componentAppliedLinearAccClamp=<5000 5000 5000>
    gravityOverride {z=-980}
    physicsBodyDefinitions [{
      boundBone {boneName=Base_blade}
      boxExtents=<8 8 8>
      constraintSetup {angularLimitsMax=<0 0 7>, angularLimitsMin=<0 -7 -7>, angularTarget=<1 0 0>}
      localJointOffset=<0.3 0 0.15>
     }]
    relativeSpaceBone {boneName=thigh_r}
   }
   n5 = Anim Dynamics - Bone: Dyn_zipper_L(ComponentPose=@n4, Alpha=1, ExternalForce="0,0,0") {
    angularDampingOverride=0.9999
    angularSpringConstant=200
    bAngularSpring=true
    bOverrideAngularDamping=true
    bUseGravityOverride=true
    alphaBoolBlend {blendOption=Linear}
    boundBone {boneName=Dyn_zipper_L}
    componentAppliedLinearAccClamp=<10000 10000 10000>
    gravityOverride {z=-980}
    physicsBodyDefinitions [{
      boundBone {boneName=Dyn_zipper_L}
      boxExtents=<5 5 5>
      constraintSetup {angularLimitsMax=<0 0 15>, angularLimitsMin=<0 -7 -15>, angularTarget=<0 0 0>}
      localJointOffset=<0.3 0 0>
     }]
   }
   n6 = Anim Dynamics - Bone: Dyn_zipper_R(ComponentPose=@n5, Alpha=1, ExternalForce="0,0,0") {
    angularDampingOverride=0.9999
    angularSpringConstant=200
    bAngularSpring=true
    bOverrideAngularDamping=true
    bUseGravityOverride=true
    alphaBoolBlend {blendOption=Linear}
    boundBone {boneName=Dyn_zipper_R}
    componentAppliedLinearAccClamp=<10000 10000 10000>
    gravityOverride {z=-980}
    physicsBodyDefinitions [{
      boundBone {boneName=Dyn_zipper_R}
      boxExtents=<5 5 5>
      constraintSetup {angularLimitsMax=<0 0 15>, angularLimitsMin=<0 -7 -15>, angularTarget=<0 0 0>}
      localJointOffset=<0.3 0 0>
     }]
   }
   n7 = Anim Dynamics - Bone: Correct_Lapel(ComponentPose=@n6, Alpha=1, ExternalForce="0,0,0") {
    angularDampingOverride=1
    angularSpringConstant=500
    bAngularSpring=true
    bOverrideAngularDamping=true
    bUseGravityOverride=true
    alphaBoolBlend {blendOption=Linear}
    boundBone {boneName=Correct_Lapel}
    componentAppliedLinearAccClamp=<5000 5000 5000>
    gravityOverride {z=-980}
    physicsBodyDefinitions [{
      boundBone {boneName=Correct_Lapel}
      boxExtents=<6 6 6>
      constraintSetup {angularLimitsMax=<0 0 5>, angularLimitsMin=<0 -7 -10>, angularTarget=<1 0 0>}
      localJointOffset=<0 0 0>
     }]
   }
   n8 = "Transform (Modify) Bone - Bone: Dyn_belt_front_03"(ComponentPose=@n7, Alpha=1, Translation="0,0,0", Rotation="0,0,0", Scale="0.9,1,1") {
    rotationMode=BMM_Additive
    scaleMode=BMM_Additive
    translationMode=BMM_Additive
    alphaBoolBlend {blendOption=Linear}
    boneToModify {boneName=Dyn_belt_front_03}
    scale {x=0.9}
   }
   n9 = "Transform (Modify) Bone - Bone: Dyn_belt_front_03"(ComponentPose=@n8, Alpha=1, Translation="0,-1.5,0", Rotation="0,0,0", Scale="1,1,1") {
    rotationMode=BMM_Additive
    scaleMode=BMM_Additive
    translationMode=BMM_Additive
    alphaBoolBlend {blendOption=Linear}
    boneToModify {boneName=Dyn_belt_front_03}
    translation {y=-1.5}
   }
   Output Pose(Result=Component To Local(ComponentPose=@n9)) {layerGroup=None}
  ]}
 EventGraph {nodes [
   n1 = Event BlueprintUpdateAnimation()
   > Set Delta Time X(Delta Time X=@n1.DeltaTimeX)
   > n2 = Gravity Override State Function DEPRECATED(
    CharacterPartAnimInstance=Self-Reference()
    GameTimeInSeconds=GetGameTimeInSeconds()
    GravityOverrideParams=Get Gravity Override Params()
    DebugJointName=dyn_hackBottomArmor
   )
   > Set Gravity Override(Gravity Override=@n2)
   n3 = Get Crouch() -> (Crouch:bool)
   n4 = Get bIsDBNO()
   > Set Alpha_RBan(Alpha_RBan:bool=AND Boolean(A=NOT Boolean(A=@n3), B=NOT Boolean(A=@n4))) -> (Output_Get:bool)
   > Set RBAN off(RBAN off:bool=NOT Boolean(A=Get bRigidBodyEnabled())) -> (Output_Get:bool)
   > Set LOD2(LOD2:bool="integer >= integer"(A=Get CurrentLOD(), B=2)) -> (Output_Get:bool)
   n5 = Get PawnVelocity()
   > Set velociry forvard(velociry forvard:bool="Not Equal (Float)"(A=BreakVector(InVec=@n5).Y)) -> (Output_Get:bool)
   > n6 = Sequence()
   n7 = NOT Boolean(A=Get bIsPlayingEmote())
   n8 = Get bIsInFrontEnd()
   > n9 = Branch(Condition=NOT Boolean(A=AND Boolean(A=@n8, B=@n7)))
   > Set FrontNotEmotr(FrontNotEmotr:double=FInterpTo(Current=Get FrontNotEmotr() -> (FrontNotEmotr:double), Target=0, DeltaTime=Get Delta Time X(), InterpSpeed=5)) -> (Output_Get:double)
   Set FrontNotEmotr(execute=@n9.else, FrontNotEmotr=FInterpTo(Current=Get FrontNotEmotr(), Target=1, DeltaTime=Get Delta Time X(), InterpSpeed=5))
   n10 = Get bIsPlayingEmote()
   n11 = Branch(execute=@n6.then_1, Condition=AND Boolean(A=Get bIsInFrontEnd(), B=NOT Boolean(A=@n10)))
   > Set My(My:double=FInterpTo(Current=Get My() -> (My:double), Target=0, DeltaTime=Get Delta Time X(), InterpSpeed=5)) -> (Output_Get:double)
   Set My(execute=@n11.else, My=FInterpTo(Current=Get My(), Target=1, DeltaTime=Get Delta Time X(), InterpSpeed=5))
   Event OnBeginCrouching()
   > Set Crouch(Crouch:bool=true) -> (Output_Get:bool)
   Event OnEndCrouching()
   > Set Crouch(Crouch=false)
   Event OnBeginPlayingEmote() {is=title K2Node_Event}
   > Set Playing Emote(Playing Emote:bool=true) -> (Output_Get:bool)
   Event OnEndPlayingEmote() {is=title K2Node_Event}
   > Set Playing Emote(Playing Emote=false)
  ]}
]
variables [
 Alpha_RBan {type=boolean}
 Crouch {type=boolean}
 Delta Time X {type=number}
 FrontNotEmotr {type=number}
 Gravity Override {type=Vector}
 Gravity Override Params {type=GravityOverrideParamsStruct}
 LOD2 {type=boolean}
 My {type=number}
 Playing Emote {type=boolean}
 RBAN off {type=boolean}
 velociry forvard {type=boolean}
]
